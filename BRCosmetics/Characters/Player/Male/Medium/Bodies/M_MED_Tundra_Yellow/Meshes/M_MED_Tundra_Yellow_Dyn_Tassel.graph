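# Physics asset for the dangling tassel and three belt bullets on the M_MED_Tundra_Yellow player cosmetic body.
# The tassel is a top-mid-bottom chain of boxes under a small master capsule, with tight swing and twist limits between links.
# The bullet capsules each attach to the pelvis with narrow limits.
# The tassel master and a large left-thigh twist capsule also constrain to the pelvis, with free angular motion.
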PhysicsAsset /BRCosmetics/Characters/Player/Male/Medium/Bodies/M_MED_Tundra_Yellow/Meshes/M_MED_Tundra_Yellow_Dyn_Tassel
bodies {
 dyn_belt_bullet_01 [{shapeType=Capsule, length=13.1087, radius=1.9115, center=<0.6627 -0.0506 1.8067>, rotation=<rot 0 87.0439 -84.5516>}]
 dyn_belt_bullet_02 [{shapeType=Capsule, length=13.2173, radius=1.89, center=<0.5942 0.1152 1.8783>, rotation=<rot 0 88.4096 -84.6092>}]
 dyn_belt_bullet_03 [{shapeType=Capsule, length=13.2474, radius=1.9991, center=<0.4391 0.2091 1.6197>, rotation=<rot 0 87.4481 -83.7546>}]
 dyn_mstr_tassel [{shapeType=Capsule, length=0.505, radius=0.505}]
 dyn_tassel_bottom [{shapeType=Box, extentX=1.9319, extentY=0.4195, extentZ=2.5264, center=<0.9477 0.0971 -0.0475>}]
 dyn_tassel_mid [{shapeType=Box, extentX=3.51, extentY=0.5, extentZ=2.51, center=<1.7963 0.0015 0.005>}]
 dyn_tassel_top [{shapeType=Box, extentX=4.01, extentY=0.51, extentZ=2.01, center=<2.0293 0.0041 -0.0081>}]
 pelvis [{shapeType=Capsule, length=0, radius=0.5, center=<1.8018 0.1808 -0.0548>, rotation=<rot 0 -5.6009 -78.6782>}]
 thigh_twist_01_l [{shapeType=Capsule, length=13.5143, radius=10.2813, center=<11.1345 -0.3475 0.1913>, rotation=<rot 90 0 0>}]
}
constraints [
 {bone1Name=dyn_tassel_top, bone2Name=dyn_mstr_tassel, swing1LimitDegrees=15, swing2LimitDegrees=15, twistLimitDegrees=5}
 {bone1Name=dyn_tassel_mid, bone2Name=dyn_tassel_top, swing1LimitDegrees=15, swing2LimitDegrees=15, twistLimitDegrees=5}
 {bone1Name=dyn_tassel_bottom, bone2Name=dyn_tassel_mid, swing1LimitDegrees=15, swing2LimitDegrees=15, twistLimitDegrees=5}
 {
  bone1Name=thigh_twist_01_l
  bone2Name=pelvis
  swing1LimitDegrees=45
  swing1Motion=Free
  swing2LimitDegrees=45
  swing2Motion=Free
  twistLimitDegrees=45
  twistMotion=Free
 }
 {bone1Name=dyn_belt_bullet_01, bone2Name=pelvis, swing1LimitDegrees=20, swing2LimitDegrees=10, twistLimitDegrees=5}
 {bone1Name=dyn_belt_bullet_02, bone2Name=pelvis, swing1LimitDegrees=20, swing2LimitDegrees=10, twistLimitDegrees=5}
 {bone1Name=dyn_belt_bullet_03, bone2Name=pelvis, swing1LimitDegrees=20, swing2LimitDegrees=10, twistLimitDegrees=5}
 {
  bone1Name=dyn_mstr_tassel
  bone2Name=pelvis
  swing1LimitDegrees=45
  swing1Motion=Free
  swing2LimitDegrees=45
  swing2Motion=Free
  twistLimitDegrees=45
  twistMotion=Free
 }
]
