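# Animation Blueprint for the M_MED_Tundra_Yellow cosmetic body, adding secondary motion to jacket, collar, gear and legs over a pose copied from the parent mesh.
# Per-update limb and head rotations become alphas for corrective bone transforms, with arm and clavicle alphas clamped under parachute.
# Weighted constraints place straps, pouches and pads, and RigidBody nodes simulate collar, tassel (off when crouching) and legs.
# A cached pose feeds two jacket simulations, selected by whether backbling is worn.
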AnimBlueprint /BRCosmetics/Characters/Player/Male/Medium/Bodies/M_MED_Tundra_Yellow/Meshes/M_MED_Tundra_Yellow_AnimBP
target_skeleton=&/M_MED_Tundra_Yellow_Skeleton
graphs [
 AnimGraph {nodes [
   n1 = RigidBody(ComponentPose=Local To Component(LocalPose=Use cached pose 'preJacketCache'())) {
    bClampLinearTranslationLimitToRefPose=true
    bOverrideWorldGravity=true
    overridePhysicsAsset=&/M_MED_Tundra_Yellow_Dyn_Jacket
    simulationSpace=BaseBoneSpace
    alphaBoolBlend {blendOption=Linear}
    baseBoneRef {boneName=spine_01}
    componentAppliedLinearAccClamp=<3000 3000 3000>
    componentLinearAccScale=<0.5 0.2 0.2>
    componentLinearVelScale=<0.5 0.2 0.2>
    overrideWorldGravity {z=-980}
   }
   n2 = RigidBody(ComponentPose=Local To Component(LocalPose=Use cached pose 'preJacketCache'())) {
    bClampLinearTranslationLimitToRefPose=true
    bOverrideWorldGravity=true
    overridePhysicsAsset=&/M_MED_Tundra_Yellow_Dyn_Jacket_No_Back
    simulationSpace=BaseBoneSpace
    alphaBoolBlend {blendOption=Linear}
    baseBoneRef {boneName=spine_01}
    componentAppliedLinearAccClamp=<3000 3000 3000>
    componentLinearAccScale=<0.5 0.2 0.2>
    componentLinearVelScale=<0.5 0.2 0.2>
    overrideWorldGravity {z=-980}
   }
   Output Pose(Result=Blend Poses by bool(BlendPose_0=Component To Local(ComponentPose=@n2), BlendPose_1=Component To Local(ComponentPose=@n1), bActiveValue=Get isWearingBackbling()) {blendType=Linear}) {layerGroup=None}
   n3 = Get thigh_rot_l()
   n4 = "Transform (Modify) Bone - Bone: pelvis_l"(
    ComponentPose=Local To Component(LocalPose=Copy Pose From Mesh() {bUseAttachedParent=true})
    Alpha=@n3
    Translation="-2.151,-0.099,0.216"
    Rotation="0,0,0"
    Scale="1.000,1.207,1.000"
   ) {
    scaleMode=BMM_Additive
    translationMode=BMM_Additive
    translationSpace=BCS_ParentBoneSpace
    alphaBoolBlend {blendOption=Linear}
    boneToModify {boneName=pelvis_l}
    scale=<1.0001 1.207>
    translation=<-2.151 -0.099 0.216>
   }
   n5 = "Transform (Modify) Bone - Bone: pelvis_mid_l"(ComponentPose=@n4, Alpha=@n3, Translation="-0.891,0.056,0.098", Rotation="0,0,0", Scale="1.523,1.000,1.000") {
    scaleMode=BMM_Additive
    translationMode=BMM_Additive
    translationSpace=BCS_ParentBoneSpace
    alphaBoolBlend {blendOption=Linear}
    boneToModify {boneName=pelvis_mid_l}
    scale {x=1.5231}
    translation=<-0.891 0.056 0.098>
   }
   n6 = "Transform (Modify) Bone - Bone: pelvis_side_l"(ComponentPose=@n5, Alpha=@n3, Translation="0,0,0", Rotation="0,0,0", Scale="1.110,1.000,1.038") {
    scaleMode=BMM_Additive
    translationMode=BMM_Additive
    translationSpace=BCS_ParentBoneSpace
    alphaBoolBlend {blendOption=Linear}
    boneToModify {boneName=pelvis_side_l}
    scale {x=1.11, z=1.038}
   }
   n7 = Get thigh_rot_r()
   n8 = "Transform (Modify) Bone - Bone: pelvis_r"(ComponentPose=@n6, Alpha=@n7, Translation="0,0,0", Rotation="0,0,0", Scale="1,1,1") {
    scaleMode=BMM_Additive
    translationMode=BMM_Additive
    translationSpace=BCS_ParentBoneSpace
    alphaBoolBlend {blendOption=Linear}
    boneToModify {boneName=pelvis_r}
   }
   n9 = "Transform (Modify) Bone - Bone: pelvis_mid_r"(ComponentPose=@n8, Alpha=@n7, Translation="0,0,0", Rotation="0,0,0", Scale="1.5,1,1.000") {
    scaleMode=BMM_Additive
    translationMode=BMM_Additive
    translationSpace=BCS_ParentBoneSpace
    alphaBoolBlend {blendOption=Linear}
    boneToModify {boneName=pelvis_mid_r}
    scale {x=1.5}
   }
   n10 = "Transform (Modify) Bone - Bone: pelvis_side_r"(ComponentPose=@n9, Alpha=@n7, Translation="0,0,0", Rotation="0,0,0", Scale="1,1,1") {
    scaleMode=BMM_Additive
    translationMode=BMM_Additive
    translationSpace=BCS_ParentBoneSpace
    alphaBoolBlend {blendOption=Linear}
    boneToModify {boneName=pelvis_side_r}
   }
   n11 = Get clav_rot_r() -> (clav_rot_r:double)
   n12 = "Transform (Modify) Bone - Bone: dyn_collar_r_MSTR"(ComponentPose=@n10, Alpha=@n11, Translation="5,-0.000,10", Rotation="-19,-0.000,-0.005", Scale="1,1,1") {
    rotationMode=BMM_Additive
    rotationSpace=BCS_BoneSpace
    translationMode=BMM_Additive
    translationSpace=BCS_BoneSpace
    alphaBoolBlend {blendOption=Linear}
    alphaScaleBiasClamp {interpSpeedDecreasing=1, interpSpeedIncreasing=1}
    boneToModify {boneName=dyn_collar_r_MSTR}
    rotation {pitch=-19, roll=-0.005}
    translation {x=5, z=10}
   }
   n13 = "Transform (Modify) Bone - Bone: dyn_collar_bk_r_MSTR"(ComponentPose=@n12, Alpha=@n11, Translation="-0.065,-0.000,-6", Rotation="-8.825,0.000,0.000", Scale="1,1,1") {
    rotationMode=BMM_Additive
    rotationSpace=BCS_BoneSpace
    translationMode=BMM_Additive
    translationSpace=BCS_BoneSpace
    alphaBoolBlend {blendOption=Linear}
    alphaScaleBiasClamp {interpSpeedDecreasing=1, interpSpeedIncreasing=1}
    boneToModify {boneName=dyn_collar_bk_r_MSTR}
    rotation {pitch=-8.825}
    translation {x=-0.065, z=-6}
   }
   n14 = "Transform (Modify) Bone - Bone: dyn_collar_fr_r_MSTR"(ComponentPose=@n13, Alpha=@n11, Translation="0,0,0", Rotation="8.097,0.000,-10.912", Scale="1,1,1") {
    rotationMode=BMM_Additive
    rotationSpace=BCS_BoneSpace
    translationMode=BMM_Additive
    translationSpace=BCS_BoneSpace
    alphaBoolBlend {blendOption=Linear}
    alphaScaleBiasClamp {interpSpeedDecreasing=1, interpSpeedIncreasing=1}
    boneToModify {boneName=dyn_collar_fr_r_MSTR}
    rotation=<rot 8.097 0 -10.9125>
   }
   n15 = Get clav_rot_l() -> (clav_rot_l:double)
   n16 = "Transform (Modify) Bone - Bone: dyn_collar_l_MSTR"(ComponentPose=@n14, Alpha=@n15, Translation="0.141,-0.000,0", Rotation="-16.890,0.002,0", Scale="1,1,1") {
    rotationMode=BMM_Additive
    rotationSpace=BCS_BoneSpace
    translationMode=BMM_Additive
    translationSpace=BCS_BoneSpace
    alphaBoolBlend {blendOption=Linear}
    alphaScaleBiasClamp {interpSpeedDecreasing=1, interpSpeedIncreasing=1}
    boneToModify {boneName=dyn_collar_l_MSTR}
    rotation {pitch=-16.89, yaw=0.002}
    translation {x=0.141}
   }
   n17 = "Transform (Modify) Bone - Bone: dyn_collar_bk_l_MSTR"(ComponentPose=@n16, Alpha=@n15, Translation="0.221,-0.233,-2", Rotation="-4.014,2.644,-14.263", Scale="1,1,1") {
    rotationMode=BMM_Additive
    rotationSpace=BCS_BoneSpace
    translationMode=BMM_Additive
    translationSpace=BCS_BoneSpace
    alphaBoolBlend {blendOption=Linear}
    alphaScaleBiasClamp {interpSpeedDecreasing=1, interpSpeedIncreasing=1}
    boneToModify {boneName=dyn_collar_bk_l_MSTR}
    rotation=<rot -4.014 2.644 -14.263>
    translation=<0.221 -0.233 -2>
   }
   n18 = "Transform (Modify) Bone - Bone: dyn_collar_fr_l_MSTR"(ComponentPose=@n17, Alpha=@n15, Translation="0.378,0.000,1.3", Rotation="-5.448,0.000,0.000", Scale="1,1,1") {
    rotationMode=BMM_Additive
    rotationSpace=BCS_BoneSpace
    translationMode=BMM_Additive
    translationSpace=BCS_BoneSpace
    alphaBoolBlend {blendOption=Linear}
    alphaScaleBiasClamp {interpSpeedDecreasing=1, interpSpeedIncreasing=1}
    boneToModify {boneName=dyn_collar_fr_l_MSTR}
    rotation {pitch=-5.448}
    translation {x=0.378, z=1.3}
   }
   n19 = Get arm_up_r() -> (arm_up_r:double)
   n20 = "Transform (Modify) Bone - Bone: dyn_collar_fr_r_MSTR"(ComponentPose=@n18, Alpha=@n19, Translation="0,0,0", Rotation="30,0.000,0.000", Scale="1,1,1") {
    rotationMode=BMM_Additive
    rotationSpace=BCS_BoneSpace
    translationSpace=BCS_BoneSpace
    alphaBoolBlend {blendOption=Linear}
    boneToModify {boneName=dyn_collar_fr_r_MSTR}
    rotation {pitch=30}
   }
   n21 = "Transform (Modify) Bone - Bone: dyn_collar_r_MSTR"(ComponentPose=@n20, Alpha=@n19, Translation="1,0.000,0", Rotation="-5.960,-0.002,-0.010", Scale="1,1,1") {
    rotationMode=BMM_Additive
    rotationSpace=BCS_BoneSpace
    translationMode=BMM_Additive
    translationSpace=BCS_BoneSpace
    alphaBoolBlend {blendOption=Linear}
    alphaScaleBiasClamp {interpSpeedDecreasing=1, interpSpeedIncreasing=1}
    boneToModify {boneName=dyn_collar_r_MSTR}
    rotation=<rot -5.96 -0.002 -0.01>
    translation {x=1}
   }
   n22 = "Transform (Modify) Bone - Bone: dyn_collar_bk_r_MSTR"(ComponentPose=@n21, Alpha=@n19, Translation="2.240,0.000,-0.865", Rotation="-22.728,-17.014,6.737", Scale="1,1,1") {
    rotationMode=BMM_Additive
    rotationSpace=BCS_BoneSpace
    translationMode=BMM_Additive
    translationSpace=BCS_BoneSpace
    alphaBoolBlend {blendOption=Linear}
    alphaScaleBiasClamp {interpSpeedDecreasing=1, interpSpeedIncreasing=1}
    boneToModify {boneName=dyn_collar_bk_r_MSTR}
    rotation=<rot -22.728 -17.014 6.737>
    translation=<2.24 0 -0.865>
   }
   n23 = "Transform (Modify) Bone - Bone: dyn_collar_back_MSTR"(ComponentPose=@n22, Alpha=Get Head Up() -> (Head Up:double), Translation="-0.5,0,1", Rotation="0,0,0", Scale="1,1,1") {
    rotationMode=BMM_Additive
    rotationSpace=BCS_BoneSpace
    translationMode=BMM_Additive
    translationSpace=BCS_BoneSpace
    alphaBoolBlend {blendOption=Linear}
    alphaScaleBiasClamp {interpSpeedDecreasing=1, interpSpeedIncreasing=1}
    boneToModify {boneName=dyn_collar_back_MSTR}
    translation {x=-0.5, z=1}
   }
   n24 = Get arm_up_l() -> (arm_up_l:double)
   n25 = "Transform (Modify) Bone - Bone: dyn_collar_fr_l_MSTR"(ComponentPose=@n23, Alpha=@n24, Translation="-0.071,-0.000,-2.410", Rotation="-8.067,0.007,-0.001", Scale="1,1,1") {
    rotationMode=BMM_Additive
    rotationSpace=BCS_BoneSpace
    translationMode=BMM_Additive
    translationSpace=BCS_BoneSpace
    alphaBoolBlend {blendOption=Linear}
    alphaScaleBiasClamp {interpSpeedDecreasing=1, interpSpeedIncreasing=1}
    boneToModify {boneName=dyn_collar_fr_l_MSTR}
    rotation=<rot -8.067 0.007 -0.001>
    translation {x=-0.071, z=-2.41}
   }
   n26 = "Transform (Modify) Bone - Bone: dyn_collar_l_MSTR"(ComponentPose=@n25, Alpha=@n24, Translation="-1.5,0.000,-3", Rotation="-10,-0.007,-0.002", Scale="1,1,1") {
    rotationMode=BMM_Additive
    rotationSpace=BCS_BoneSpace
    translationMode=BMM_Additive
    translationSpace=BCS_BoneSpace
    alphaBoolBlend {blendOption=Linear}
    alphaScaleBiasClamp {interpSpeedDecreasing=1, interpSpeedIncreasing=1}
    boneToModify {boneName=dyn_collar_l_MSTR}
    rotation=<rot -10 -0.007 -0.002>
    translation {x=-1.5, z=-3}
   }
   n27 = "Transform (Modify) Bone - Bone: dyn_collar_bk_l_MSTR"(ComponentPose=@n26, Alpha=@n24, Translation="0.425,1.058,-3.023", Rotation="21.297,0.004,0.000", Scale="1,1,1") {
    rotationMode=BMM_Additive
    rotationSpace=BCS_BoneSpace
    translationMode=BMM_Additive
    translationSpace=BCS_BoneSpace
    alphaBoolBlend {blendOption=Linear}
    alphaScaleBiasClamp {interpSpeedDecreasing=1, interpSpeedIncreasing=1}
    boneToModify {boneName=dyn_collar_bk_l_MSTR}
    rotation {pitch=21.297, yaw=0.004}
    translation=<0.4253 1.058 -3.0233>
   }
   n28 = "Constraint - dyn_strap_clip_fr_l_MSTR (PPPPPP)"!(
    ComponentPose=@n27
    Alpha=1
    ConstraintWeights_0=0.4021
    ConstraintWeights_1:float=0.0663
    ConstraintWeights_2:float=0.1299
    ConstraintWeights_3:float=0.174
    ConstraintWeights_4:float=0.0925
    ConstraintWeights_5:float=0.1352
   ) -> (Pose) {
    alphaBoolBlend {blendOption=Linear}
    boneToModify {boneName=dyn_strap_clip_fr_l_MSTR}
    constraintSetup [
     {targetBone {boneName=clavicle_l}}
     {targetBone {boneName=spine_02}}
     {targetBone {boneName=spine_03}}
     {targetBone {boneName=spine_04}}
     {targetBone {boneName=spine_05}}
     {targetBone {boneName=upperarm_twist_01_l}}
    ]
    constraintWeights [1,1,1,1,1,1]
   }
   n29 = "Constraint - dyn_strap_clip_fr_r_MSTR (PPPPPP)"!(
    ComponentPose=@n28
    Alpha=1
    ConstraintWeights_0=0.1344
    ConstraintWeights_1:float=0.0671
    ConstraintWeights_2:float=0.1311
    ConstraintWeights_3:float=0.1756
    ConstraintWeights_4:float=0.0909
    ConstraintWeights_5:float=0.4009
   ) -> (Pose) {
    alphaBoolBlend {blendOption=Linear}
    boneToModify {boneName=dyn_strap_clip_fr_r_MSTR}
    constraintSetup [
     {targetBone {boneName=upperarm_twist_01_r}}
     {targetBone {boneName=spine_02}}
     {targetBone {boneName=spine_03}}
     {targetBone {boneName=spine_04}}
     {targetBone {boneName=spine_05}}
     {targetBone {boneName=clavicle_r}}
    ]
    constraintWeights [1,1,1,1,1,1]
   }
   n30 = "Constraint - dyn_strap_clip_bk_l_MSTR (PPPPP)"!(
    ComponentPose=@n29
    Alpha=1
    ConstraintWeights_0=0.1812
    ConstraintWeights_1:float=0.0801
    ConstraintWeights_2:float=0.0469
    ConstraintWeights_3:float=0.5755
    ConstraintWeights_4:float=0.1163
   ) -> (Pose) {
    alphaBoolBlend {blendOption=Linear}
    boneToModify {boneName=dyn_strap_clip_bk_l_MSTR}
    constraintSetup [
     {targetBone {boneName=spine_03}}
     {targetBone {boneName=spine_04}}
     {targetBone {boneName=spine_05}}
     {targetBone {boneName=clavicle_l}}
     {targetBone {boneName=upperarm_twist_01_l}}
    ]
    constraintWeights [1,1,1,1,1]
   }
   n31 = "Constraint - dyn_strap_clip_bk_r_MSTR (PPPPP)"!(
    ComponentPose=@n30
    Alpha=1
    ConstraintWeights_0=0.1158
    ConstraintWeights_1:float=0.5736
    ConstraintWeights_2:float=0.1832
    ConstraintWeights_3:float=0.0828
    ConstraintWeights_4:float=0.0446
   ) -> (Pose) {
    alphaBoolBlend {blendOption=Linear}
    boneToModify {boneName=dyn_strap_clip_bk_r_MSTR}
    constraintSetup [
     {targetBone {boneName=upperarm_twist_01_r}}
     {targetBone {boneName=clavicle_r}}
     {targetBone {boneName=spine_03}}
     {targetBone {boneName=spine_04}}
     {targetBone {boneName=spine_05}}
    ]
    constraintWeights [1,1,1,1,1]
   }
   n32 = "Constraint - dyn_chest_zipper_r_MSTR (PPPPPPP)"!(
    ComponentPose=@n31
    Alpha=1
    ConstraintWeights_0=0.0794
    ConstraintWeights_1:float=0.1875
    ConstraintWeights_2:float=0.3748
    ConstraintWeights_3:float=0.2021
    ConstraintWeights_4:float=0.0121
    ConstraintWeights_5:float=0.1244
    ConstraintWeights_6:float=0.0198
   ) -> (Pose) {
    alphaBoolBlend {blendOption=Linear}
    boneToModify {boneName=dyn_chest_zipper_r_MSTR}
    constraintSetup [
     {targetBone {boneName=spine_01}}
     {targetBone {boneName=spine_02}}
     {targetBone {boneName=spine_03}}
     {targetBone {boneName=spine_04}}
     {targetBone {boneName=spine_05}}
     {targetBone {boneName=clavicle_r}}
     {targetBone {boneName=pelvis}}
    ]
    constraintWeights [1,1,1,1,1,1,1]
   }
   n33 = "Constraint - dyn_collar_zipper_MSTR (PPPPPPP)"!(
    ComponentPose=@n32
    Alpha=1
    ConstraintWeights_0=0.1041
    ConstraintWeights_1:float=0.0233
    ConstraintWeights_2:float=0.1428
    ConstraintWeights_3:float=0.2469
    ConstraintWeights_4:float=0.3908
    ConstraintWeights_5:float=0.0715
    ConstraintWeights_6:float=0.0206
   ) -> (Pose) {
    alphaBoolBlend {blendOption=Linear}
    boneToModify {boneName=dyn_collar_zipper_MSTR}
    constraintSetup [
     {targetBone {boneName=clavicle_l}}
     {targetBone {boneName=spine_01}}
     {targetBone {boneName=spine_02}}
     {targetBone {boneName=spine_03}}
     {targetBone {boneName=spine_04}}
     {targetBone {boneName=spine_05}}
     {targetBone {boneName=clavicle_r}}
    ]
    constraintWeights [1,1,1,1,1,1,1]
   }
   n34 = "Constraint - dyn_armband_knot_MSTR (PPPPP)"!(
    ComponentPose=@n33
    Alpha=1
    ConstraintWeights_0=0.1023
    ConstraintWeights_1:float=0.4107
    ConstraintWeights_2:float=0.0197
    ConstraintWeights_3:float=0.2103
    ConstraintWeights_4:float=0.2569
   ) -> (Pose) {
    alphaBoolBlend {blendOption=Linear}
    boneToModify {boneName=dyn_armband_knot_MSTR}
    constraintSetup [
     {targetBone {boneName=clavicle_r}}
     {targetBone {boneName=upperarm_r}}
     {targetBone {boneName=spine_03}}
     {targetBone {boneName=upperarm_twist_02_r}}
     {targetBone {boneName=upperarm_twist_01_r}}
    ]
    constraintWeights [1,1,1,1,1]
   }
   n35 = "Constraint - dyn_belt_bullet_01 (PPPPPP)"!(
    ComponentPose=@n34
    Alpha=1
    ConstraintWeights_0=0.3893
    ConstraintWeights_1:float=0.1443
    ConstraintWeights_2:float=0.0624
    ConstraintWeights_3:float=0.0001
    ConstraintWeights_4:float=0
    ConstraintWeights_5:float=0.4039
   ) -> (Pose) {
    alphaBoolBlend {blendOption=Linear}
    boneToModify {boneName=dyn_belt_bullet_01}
    constraintSetup [
     {targetBone {boneName=spine_01}}
     {targetBone {boneName=spine_02}}
     {targetBone {boneName=spine_03}}
     {targetBone {boneName=spine_04}}
     {targetBone {boneName=clavicle_r}}
     {targetBone {boneName=pelvis}}
    ]
    constraintWeights [1,1,1,1,1,1]
   }
   n36 = "Constraint - dyn_belt_bullet_02 (PPPPPP)"!(
    ComponentPose=@n35
    Alpha=1
    ConstraintWeights_0=0.388
    ConstraintWeights_1:float=0.1398
    ConstraintWeights_2:float=0.064
    ConstraintWeights_3:float=0.0002
    ConstraintWeights_4:float=0.0001
    ConstraintWeights_5:float=0.408
   ) -> (Pose) {
    alphaBoolBlend {blendOption=Linear}
    boneToModify {boneName=dyn_belt_bullet_02}
    constraintSetup [
     {targetBone {boneName=spine_01}}
     {targetBone {boneName=spine_02}}
     {targetBone {boneName=spine_03}}
     {targetBone {boneName=spine_04}}
     {targetBone {boneName=clavicle_r}}
     {targetBone {boneName=pelvis}}
    ]
    constraintWeights [1,1,1,1,1,1]
   }
   n37 = "Constraint - dyn_belt_bullet_03 (PPPPP)"!(
    ComponentPose=@n36
    Alpha=1
    ConstraintWeights_0=0.3741
    ConstraintWeights_1:float=0.1365
    ConstraintWeights_2:float=0.0683
    ConstraintWeights_3:float=0.0005
    ConstraintWeights_4:float=0.4206
   ) -> (Pose) {
    alphaBoolBlend {blendOption=Linear}
    boneToModify {boneName=dyn_belt_bullet_03}
    constraintSetup [
     {targetBone {boneName=spine_01}}
     {targetBone {boneName=spine_02}}
     {targetBone {boneName=spine_03}}
     {targetBone {boneName=spine_04}}
     {targetBone {boneName=pelvis}}
    ]
    constraintWeights [1,1,1,1,1]
   }
   n38 = "Constraint - dyn_belt_back_pouch (PPPP)"!(
    ComponentPose=@n37
    Alpha=1
    ConstraintWeights_0=0.1735
    ConstraintWeights_1:float=0.0357
    ConstraintWeights_2:float=0.0459
    ConstraintWeights_3:float=0.7448
   ) -> (Pose) {
    alphaBoolBlend {blendOption=Linear}
    boneToModify {boneName=dyn_belt_back_pouch}
    constraintSetup [{targetBone {boneName=spine_01}}, {targetBone {boneName=spine_02}}, {targetBone {boneName=spine_03}}, {targetBone {boneName=pelvis}}]
    constraintWeights [1,1,1,1]
   }
   n39 = "Constraint - dyn_hip_pouch_bk_r (PPP)"!(ComponentPose=@n38, Alpha=1, ConstraintWeights_0=0.0155, ConstraintWeights_1:float=0.3262, ConstraintWeights_2:float=0.6583) -> (Pose) {
    alphaBoolBlend {blendOption=Linear}
    boneToModify {boneName=dyn_hip_pouch_bk_r}
    constraintSetup [{targetBone {boneName=thigh_r}}, {targetBone {boneName=pelvis}}, {targetBone {boneName=thigh_twist_01_r}}]
    constraintWeights [1,1,1]
   }
   n40 = "Constraint - dyn_hip_pouch_fr_r (PP)"!(ComponentPose=@n39, Alpha=1, ConstraintWeights_0=0.2777, ConstraintWeights_1:float=0.7223) -> (Pose) {
    alphaBoolBlend {blendOption=Linear}
    boneToModify {boneName=dyn_hip_pouch_fr_r}
    constraintSetup [{targetBone {boneName=pelvis}}, {targetBone {boneName=thigh_twist_01_r}}]
    constraintWeights [1,1]
   }
   n41 = "Constraint - dyn_hip_pouch_fr_l (PPP)"!(ComponentPose=@n40, Alpha=1, ConstraintWeights_0=0.0775, ConstraintWeights_1:float=0.8334, ConstraintWeights_2:float=0.089) -> (Pose) {
    alphaBoolBlend {blendOption=Linear}
    boneToModify {boneName=dyn_hip_pouch_fr_l}
    constraintSetup [{targetBone {boneName=pelvis}}, {targetBone {boneName=thigh_twist_01_l}}, {targetBone {boneName=thigh_l}}]
    constraintWeights [1,1,1]
   }
   n42 = "Constraint - dyn_kneepad_l (PP)"!(ComponentPose=@n41, Alpha=1, ConstraintWeights_0=0.7, ConstraintWeights_1:float=0.3) -> (Pose) {
    alphaBoolBlend {blendOption=Linear}
    boneToModify {boneName=dyn_kneepad_l}
    constraintSetup [{targetBone {boneName=calf_l}}, {targetBone {boneName=thigh_l}}]
    constraintWeights [1,1]
   }
   n43 = "Constraint - dyn_kneepad_r (PP)"!(ComponentPose=@n42, Alpha=1, ConstraintWeights_0=0.3, ConstraintWeights_1:float=0.7) -> (Pose) {
    alphaBoolBlend {blendOption=Linear}
    boneToModify {boneName=dyn_kneepad_r}
    constraintSetup [{targetBone {boneName=calf_r}}, {targetBone {boneName=thigh_r}}]
    constraintWeights [1,1]
   }
   n44 = "Transform (Modify) Bone - Bone: dyn_kneepad_l"(ComponentPose=@n43, Alpha=Get knee_rot_l() -> (knee_rot_l:double), Translation="0.000,-3,0.000", Rotation="0,0,0", Scale="1,1,1") {
    rotationMode=BMM_Additive
    rotationSpace=BCS_BoneSpace
    translationMode=BMM_Additive
    translationSpace=BCS_BoneSpace
    alphaBoolBlend {blendOption=Linear}
    alphaScaleBiasClamp {interpSpeedDecreasing=1, interpSpeedIncreasing=1}
    boneToModify {boneName=dyn_kneepad_l}
    translation {y=-3}
   }
   n45 = "Transform (Modify) Bone - Bone: dyn_kneepad_r"(ComponentPose=@n44, Alpha=Get knee_rot_r() -> (knee_rot_r:double), Translation="0.000,3,0.000", Rotation="0,0,0", Scale="1,1,1") {
    rotationMode=BMM_Additive
    rotationSpace=BCS_BoneSpace
    translationMode=BMM_Additive
    translationSpace=BCS_BoneSpace
    alphaBoolBlend {blendOption=Linear}
    alphaScaleBiasClamp {interpSpeedDecreasing=1, interpSpeedIncreasing=1}
    boneToModify {boneName=dyn_kneepad_r}
    translation {y=3}
   }
   n46 = "Transform (Modify) Bone - Bone: dyn_collar_back_MSTR"(ComponentPose=@n45, Alpha=1, Translation="0.000,-2,-1", Rotation="0,0,0", Scale="1,1,1") {translationMode=BMM_Additive, alphaBoolBlend {blendOption=Linear}, boneToModify {boneName=dyn_collar_back_MSTR}, translation {y=-2, z=-1}}
   n47 = RigidBody(ComponentPose=@n46) {
    bClampLinearTranslationLimitToRefPose=true
    bOverrideWorldGravity=true
    overridePhysicsAsset=&/M_MED_Tundra_Yellow_Dyn_Collar
    simulationSpace=BaseBoneSpace
    alphaBoolBlend {blendOption=Linear}
    baseBoneRef {boneName=spine_04}
    componentAppliedLinearAccClamp=<10 10 10>
    componentLinearAccScale=<0.1 0.1 0.1>
    componentLinearVelScale=<3 3 3>
    overrideWorldGravity {z=-980}
   }
   n48 = RigidBody!(ComponentPose=@n47, bAlphaBoolEnabled:bool=NOT Boolean(A=Get bIsCrouching()), ExternalForce) -> (Pose) {
    alphaInputType=Bool
    bClampLinearTranslationLimitToRefPose=true
    bOverrideWorldGravity=true
    overridePhysicsAsset=&/M_MED_Tundra_Yellow_Dyn_Tassel
    simulationSpace=BaseBoneSpace
    alphaBoolBlend {blendOption=Linear}
    baseBoneRef {boneName=pelvis}
    componentAppliedLinearAccClamp=<3000 3000 3000>
    componentLinearAccScale=<0.4 0.3 0.3>
    componentLinearVelScale=<0.4 0.3 0.3>
    overrideWorldGravity {z=-980}
   }
   n49 = RigidBody!(ComponentPose=@n48, Alpha:float, ExternalForce) -> (Pose) {
    bClampLinearTranslationLimitToRefPose=true
    bOverrideWorldGravity=true
    overridePhysicsAsset=&/M_MED_Tundra_Yellow_Dyn_Legs
    simulationSpace=BaseBoneSpace
    alphaBoolBlend {blendOption=Linear}
    baseBoneRef {boneName=thigh_l}
    componentAppliedLinearAccClamp=<3000 3000 3000>
    componentLinearAccScale=<0.12 0.4 0.4>
    componentLinearVelScale=<0.12 0.4 0.4>
    overrideWorldGravity {z=-980}
   }
   n50 = "Constraint - dyn_backpack (PPPPP)"!(
    ComponentPose=@n49
    Alpha=1
    ConstraintWeights_0=0.2133
    ConstraintWeights_1:float=0.3414
    ConstraintWeights_2:float=0.1453
    ConstraintWeights_3:float=0.0432
    ConstraintWeights_4:float=0.2567
   ) -> (Pose) {
    alphaBoolBlend {blendOption=Linear}
    boneToModify {boneName=dyn_backpack}
    constraintSetup [
     {targetBone {boneName=clavicle_r}}
     {targetBone {boneName=spine_03}}
     {targetBone {boneName=spine_04}}
     {targetBone {boneName=spine_05}}
     {targetBone {boneName=clavicle_l}}
    ]
    constraintWeights [1,1,1,1,1]
   }
   Save cached pose 'preJacketCache'(Pose=Component To Local(ComponentPose=@n50)) {cacheName=preJacketCache}
  ]}
 EventGraph {nodes [
   Delay(Duration=0.125)
   > ResetDynamics(InTeleportType=TeleportPhysics)
   > ResetDynamics(InTeleportType=ResetPhysics)
   Event BlueprintUpdateAnimation()
   > n1 = Sequence!(execute) -> (then_0, then_1, then_2:Exec)
   n2 = BreakTransform(InTransform=GetDeltaTransformFromRefPose(self=GetOwnerSkeletalMeshComponent(), BoneName=upperarm_r, BaseName=spine_05))
   > Set arm_up_r(arm_up_r:double=MapRangeClamped(Value=BreakRotator(InRot=@n2.Rotation).Yaw, InRangeA=60, InRangeB=100)) -> (Output_Get:double)
   n3 = BreakTransform(InTransform=GetDeltaTransformFromRefPose(self=GetOwnerSkeletalMeshComponent(), BoneName=upperarm_l, BaseName=spine_05))
   > Set arm_up_l(arm_up_l:double=MapRangeClamped(Value=BreakRotator(InRot=@n3.Rotation).Yaw, InRangeA=50, InRangeB=105)) -> (Output_Get:double)
   n4 = BreakTransform(InTransform=GetDeltaTransformFromRefPose(self=GetOwnerSkeletalMeshComponent(), BoneName=head, BaseName=spine_05))
   > Set Head Up(Head Up:double=MapRangeClamped(Value=BreakRotator(InRot=@n4.Rotation).Pitch, InRangeA=0, InRangeB=15)) -> (Output_Get:double)
   n5 = BreakTransform(InTransform=GetDeltaTransformFromRefPose(self=GetOwnerSkeletalMeshComponent(), BoneName=calf_l, BaseName=thigh_l))
   > Set knee_rot_l(knee_rot_l:double=MapRangeClamped(Value=BreakRotator(InRot=@n5.Rotation).Yaw, InRangeA=13, InRangeB=90)) -> (Output_Get:double)
   n6 = BreakTransform(InTransform=GetDeltaTransformFromRefPose(self=GetOwnerSkeletalMeshComponent(), BoneName=calf_r, BaseName=thigh_r))
   > Set knee_rot_r(knee_rot_r:double=MapRangeClamped(Value=BreakRotator(InRot=@n6.Rotation).Yaw, InRangeA=13, InRangeB=90)) -> (Output_Get:double)
   n7 = BreakTransform(InTransform=GetDeltaTransformFromRefPose(self=GetOwnerSkeletalMeshComponent(), BoneName=thigh_r, BaseName=pelvis))
   > Set thigh_rot_r(thigh_rot_r=MapRangeClamped(Value=BreakRotator(InRot=@n7.Rotation).Yaw, InRangeA=0, InRangeB=-90))
   n8 = BreakTransform(InTransform=GetDeltaTransformFromRefPose(self=GetOwnerSkeletalMeshComponent(), BoneName=thigh_l, BaseName=pelvis))
   > Set thigh_rot_l(thigh_rot_l=MapRangeClamped(Value=BreakRotator(InRot=@n8.Rotation).Yaw, InRangeA=-15, InRangeB=-110))
   n9 = BreakTransform(InTransform=GetDeltaTransformFromRefPose(self=GetOwnerSkeletalMeshComponent(), BoneName=clavicle_r, BaseName=spine_05))
   > Set clav_rot_r(clav_rot_r:double=MapRangeClamped(Value=BreakRotator(InRot=@n9.Rotation).Pitch, InRangeA=9, InRangeB=35)) -> (Output_Get:double)
   n10 = BreakTransform(InTransform=GetDeltaTransformFromRefPose(self=GetOwnerSkeletalMeshComponent(), BoneName=clavicle_l, BaseName=spine_05))
   > Set clav_rot_l(clav_rot_l:double=MapRangeClamped(Value=BreakRotator(InRot=@n10.Rotation).Pitch, InRangeA=6, InRangeB=35)) -> (Output_Get:double)
   Do Once(execute=@n1.then_1)
   n11 = GetPartSkeletalMeshComponent()
   n12 = FindChildCommponentFromAttachSocket!(
    self:Custom Character Part Anim Instance Object Reference
    Parent:Skeletal Mesh Component Object Reference=@n11
    bIncludeAllDescendants:bool=false
    AttachSocket:Name=Cape
   ) -> (ReturnValue:Skeletal Mesh Component Object Reference)
   n13 = FindChildCommponentFromAttachSocket(Parent=@n11, bIncludeAllDescendants=false, AttachSocket=backpack_br)
   > Set isWearingBackbling(isWearingBackbling=OR Boolean(A="Not Equal (Object)"!(A:Object Reference=@n12, B:Object Reference) -> (ReturnValue:bool), B="Not Equal (Object)"(A=@n13)))
   Branch(execute=@n1.then_2, Condition=Get bIsParachuteOpened())
   > Set arm_up_l(arm_up_l="Clamp (Float)"(Value=Get arm_up_l(), Min=0, Max=0.05))
   > Set arm_up_r(arm_up_r="Clamp (Float)"(Value=Get arm_up_r(), Min=0, Max=0.05))
   > Set clav_rot_r(clav_rot_r="Clamp (Float)"(Value=Get clav_rot_r(), Min=0, Max=0.05))
   > Set clav_rot_l(clav_rot_l="Clamp (Float)"(Value=Get clav_rot_l(), Min=0, Max=0.05))
   K2Node_Event_1 = CustomEvent() {is=title K2Node_CustomEvent}
   > Set crouching alpha(crouching alpha:double=1) -> (Output_Get:double)
   K2Node_Event_2 = CustomEvent_1() {is=title K2Node_CustomEvent}
   > Set crouching alpha(crouching alpha=0)
   Event OnBeginTargeting() {is=title K2Node_Event}
   > Set targeting_alpha(targeting_alpha:double=1) -> (Output_Get:double)
   Event OnEndTargeting() {is=title K2Node_Event}
   > Set targeting_alpha(targeting_alpha=0)
   > Delay(Duration=0.125)
   > ResetDynamics(InTeleportType=TeleportPhysics)
   > ResetDynamics(InTeleportType=ResetPhysics)
   Event OnBeginCrouching()
   Event OnEndCrouching()
  ]}
]
variables [
 arm_up_l {type=number}
 arm_up_r {type=number}
 clav_rot_l {type=number}
 clav_rot_r {type=number}
 crouching alpha {type=number}
 Head Up {type=number}
 isWearingBackbling {type=boolean}
 knee_rot_l {type=number}
 knee_rot_r {type=number}
 targeting_alpha {type=number}
 thigh_rot_l {type=number}
 thigh_rot_r {type=number}
]
