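# Post-corrects the hood and collar of the Tundra Yellow male cosmetic to follow shoulder and arm motion.
# Each update maps clavicle pitch and upper-arm yaw relative to spine_05 into alphas, clamped low by a parachute-open branch.
# Those alphas drive additive bone offsets on the parent-copied pose, then RigidBody simulates in neck_02 space with a physics asset override.
# The Head Up back-collar node is disconnected from the output.
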AnimBlueprint /BRCosmetics/Characters/Player/Male/Medium/Bodies/M_MED_Tundra_Yellow/Meshes/Parts/M_MED_Tundra_Yellow_Hood_FaceAcc_AnimBP2
target_skeleton=&/M_MED_Tundra_Yellow_Hood_FaceAcc_Skeleton
graphs [
 AnimGraph {nodes [
   "Transform (Modify) Bone - Bone: dyn_collar_back_MSTR"(Alpha=Get Head Up() -> (Head Up:double), Translation="-2,0,4", Rotation="0,0,0", Scale="1,1,1") {
    rotationMode=BMM_Additive
    rotationSpace=BCS_BoneSpace
    translationMode=BMM_Additive
    translationSpace=BCS_BoneSpace
    alphaBoolBlend {blendOption=Linear}
    alphaScaleBiasClamp {interpSpeedDecreasing=1, interpSpeedIncreasing=1}
    boneToModify {boneName=dyn_collar_back_MSTR}
    translation {x=-2, z=4}
   }
   n1 = Get clav_rot_r() -> (clav_rot_r:double)
   n2 = "Transform (Modify) Bone - Bone: dyn_collar_r_MSTR"(
    ComponentPose=Local To Component(LocalPose=Copy Pose From Mesh() {bUseAttachedParent=true})
    Alpha=@n1
    Translation="-4,0,-2"
    Rotation="-5,-0.000,-10"
    Scale="1,1,1"
   ) {
    rotationMode=BMM_Additive
    rotationSpace=BCS_BoneSpace
    translationMode=BMM_Additive
    translationSpace=BCS_BoneSpace
    alphaBoolBlend {blendOption=Linear}
    alphaScaleBiasClamp {interpSpeedDecreasing=1, interpSpeedIncreasing=1}
    boneToModify {boneName=dyn_collar_r_MSTR}
    rotation {pitch=-5, roll=-10}
    translation {x=-4, z=-2}
   }
   n3 = Get arm_up_r() -> (arm_up_r:double)
   n4 = "Transform (Modify) Bone - Bone: dyn_collar_r_MSTR"(ComponentPose=@n2, Alpha=@n3, Translation="-6,0,-3", Rotation="-10,0,0", Scale="1,1,1") {
    rotationMode=BMM_Additive
    rotationSpace=BCS_BoneSpace
    translationMode=BMM_Additive
    translationSpace=BCS_BoneSpace
    alphaBoolBlend {blendOption=Linear}
    alphaScaleBiasClamp {interpSpeedDecreasing=1, interpSpeedIncreasing=1}
    boneToModify {boneName=dyn_collar_r_MSTR}
    rotation {pitch=-10}
    translation {x=-6, z=-3}
   }
   n5 = "Transform (Modify) Bone - Bone: dyn_collar_fr_r_MSTR"(ComponentPose=@n4, Alpha=@n1, Translation="-2,0,-1", Rotation="0,0,-20", Scale="1,1,1") {
    rotationMode=BMM_Additive
    rotationSpace=BCS_BoneSpace
    translationMode=BMM_Additive
    translationSpace=BCS_BoneSpace
    alphaBoolBlend {blendOption=Linear}
    alphaScaleBiasClamp {interpSpeedDecreasing=1, interpSpeedIncreasing=1}
    boneToModify {boneName=dyn_collar_fr_r_MSTR}
    rotation {roll=-20}
    translation {x=-2, z=-1}
   }
   n6 = "Transform (Modify) Bone - Bone: dyn_collar_fr_r_MSTR"(ComponentPose=@n5, Alpha=@n3, Translation="-2,0,-1", Rotation="0,0,-20", Scale="1,1,1") {
    rotationMode=BMM_Additive
    rotationSpace=BCS_BoneSpace
    translationMode=BMM_Additive
    translationSpace=BCS_BoneSpace
    alphaBoolBlend {blendOption=Linear}
    alphaScaleBiasClamp {interpSpeedDecreasing=1, interpSpeedIncreasing=1}
    boneToModify {boneName=dyn_collar_fr_r_MSTR}
    rotation {roll=-20}
    translation {x=-2, z=-1}
   }
   n7 = "Transform (Modify) Bone - Bone: dyn_hood_r_MSTR"(ComponentPose=@n6, Alpha=@n1, Translation="0,0,0", Rotation="0,0,0", Scale="1,1,1") {
    rotationMode=BMM_Additive
    rotationSpace=BCS_BoneSpace
    translationMode=BMM_Additive
    translationSpace=BCS_BoneSpace
    alphaBoolBlend {blendOption=Linear}
    alphaScaleBiasClamp {interpSpeedDecreasing=1, interpSpeedIncreasing=1}
    boneToModify {boneName=dyn_hood_r_MSTR}
   }
   n8 = "Transform (Modify) Bone - Bone: dyn_hood_r_MSTR"(ComponentPose=@n7, Alpha=@n3, Translation="0,0,2.5", Rotation="0,0,0", Scale="1,1,1") {
    rotationMode=BMM_Additive
    rotationSpace=BCS_BoneSpace
    translationMode=BMM_Additive
    translationSpace=BCS_BoneSpace
    alphaBoolBlend {blendOption=Linear}
    alphaScaleBiasClamp {interpSpeedDecreasing=1, interpSpeedIncreasing=1}
    boneToModify {boneName=dyn_hood_r_MSTR}
    translation {z=2.5}
   }
   n9 = "Transform (Modify) Bone - Bone: dyn_collar_l_MSTR"(ComponentPose=@n8, Alpha=Get clav_rot_l() -> (clav_rot_l:double), Translation="0.2,-0.000,0", Rotation="-10,0.002,0", Scale="1,1,1") {
    rotationMode=BMM_Additive
    rotationSpace=BCS_BoneSpace
    translationMode=BMM_Additive
    translationSpace=BCS_BoneSpace
    alphaBoolBlend {blendOption=Linear}
    alphaScaleBiasClamp {interpSpeedDecreasing=1, interpSpeedIncreasing=1}
    boneToModify {boneName=dyn_collar_l_MSTR}
    rotation {pitch=-10, yaw=0.002}
    translation {x=0.2}
   }
   n10 = "Transform (Modify) Bone - Bone: dyn_collar_l_MSTR"(ComponentPose=@n9, Alpha=Get arm_up_l() -> (arm_up_l:double), Translation="-15,0.000,-7", Rotation="-22.718,-0.007,-0.002", Scale="1,1,1") {
    rotationMode=BMM_Additive
    rotationSpace=BCS_BoneSpace
    translationMode=BMM_Additive
    translationSpace=BCS_BoneSpace
    alphaBoolBlend {blendOption=Linear}
    alphaScaleBiasClamp {interpSpeedDecreasing=1, interpSpeedIncreasing=1}
    boneToModify {boneName=dyn_collar_l_MSTR}
    rotation=<rot -22.718 -0.007 -0.002>
    translation {x=-15, z=-7}
   }
   n11 = RigidBody(ComponentPose=@n10) {
    bClampLinearTranslationLimitToRefPose=true
    bOverrideWorldGravity=true
    overridePhysicsAsset=&/M_MED_Tundra_Yellow_Hood_FaceAcc_Physics
    simulationSpace=BaseBoneSpace
    alphaBoolBlend {blendOption=Linear}
    baseBoneRef {boneName=neck_02}
    componentAppliedLinearAccClamp=<3000 3000 3000>
    componentLinearAccScale=<0.5 0.4 0.4>
    componentLinearVelScale=<0.5 0.4 0.4>
    overrideWorldGravity {z=-980}
   }
   Output Pose(Result=Component To Local(ComponentPose=@n11)) {layerGroup=None}
   "Clamp (Float)"(Value=0, Min=0, Max=0.15)
   "To Float (Boolean)"(InBool=false)
   "float > float"(A=0, B=0) {is=title K2Node_CallFunction}
   Get bIsParachuteOpened()
  ]}
 EventGraph {nodes [
   Event BlueprintUpdateAnimation()
   n1 = BreakTransform(InTransform=GetDeltaTransformFromRefPose(self=GetOwnerSkeletalMeshComponent(), BoneName=upperarm_r, BaseName=spine_05))
   > Set arm_up_r(arm_up_r:double=MapRangeClamped(Value=BreakRotator(InRot=@n1.Rotation).Yaw, InRangeA=60, InRangeB=120)) -> (Output_Get:double)
   n2 = BreakTransform(InTransform=GetDeltaTransformFromRefPose(self=GetOwnerSkeletalMeshComponent(), BoneName=upperarm_l, BaseName=spine_05))
   > Set arm_up_l(arm_up_l:double=MapRangeClamped(Value=BreakRotator(InRot=@n2.Rotation).Yaw, InRangeA=50, InRangeB=120)) -> (Output_Get:double)
   n3 = BreakTransform(InTransform=GetDeltaTransformFromRefPose(self=GetOwnerSkeletalMeshComponent(), BoneName=clavicle_r, BaseName=spine_05))
   > Set clav_rot_r(clav_rot_r:double=MapRangeClamped(Value=BreakRotator(InRot=@n3.Rotation).Pitch, InRangeA=9, InRangeB=35)) -> (Output_Get:double)
   n4 = BreakTransform(InTransform=GetDeltaTransformFromRefPose(self=GetOwnerSkeletalMeshComponent(), BoneName=clavicle_l, BaseName=spine_05))
   > Set clav_rot_l(clav_rot_l:double=MapRangeClamped(Value=BreakRotator(InRot=@n4.Rotation).Pitch, InRangeA=6, InRangeB=35)) -> (Output_Get:double)
   > Branch(Condition=Get bIsParachuteOpened())
   > Set arm_up_l(arm_up_l="Clamp (Float)"(Value=Get arm_up_l(), Min=0, Max=0.05))
   > Set arm_up_r(arm_up_r="Clamp (Float)"(Value=Get arm_up_r(), Min=0, Max=0.05))
   > Set clav_rot_r(clav_rot_r="Clamp (Float)"(Value=Get clav_rot_r(), Min=0, Max=0.05))
   > Set clav_rot_l(clav_rot_l="Clamp (Float)"(Value=Get clav_rot_l(), Min=0, Max=0.05))
   TryGetPawnOwner()
  ]}
]
variables [
 arm_up_l {type=number}
 arm_up_r {type=number}
 clav_rot_l {type=number}
 clav_rot_r {type=number}
 Head Up {type=number}
]
